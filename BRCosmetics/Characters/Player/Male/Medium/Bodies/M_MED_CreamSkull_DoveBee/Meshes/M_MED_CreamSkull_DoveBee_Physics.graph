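# Physics asset for the M_MED_CreamSkull_DoveBee male body cosmetic, simulating swinging jacket flaps.
# Capsules on pelvis, spine_01, thighs and two pocket bones form body volumes around the jacket.
# Six jacket base bones lock to spine_01, each anchoring a front, side or back flap with limited swing and narrow twist.
# Front flaps use two angled capsules each, while pocket constraints to the pelvis leave rotation free.
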PhysicsAsset /BRCosmetics/Characters/Player/Male/Medium/Bodies/M_MED_CreamSkull_DoveBee/Meshes/M_MED_CreamSkull_DoveBee_Physics
bodies {
 dyn_jacket_bk_l [{shapeType=Capsule, length=15, radius=3, shapeName=dyn_jacket_bk_l_capsule, center=<15 0 0>, rotation=<rot 0 0 90>}]
 dyn_jacket_bk_l_base [{shapeType=Capsule, length=2, radius=1, shapeName=dyn_jacket_bk_l_base_capsule}]
 dyn_jacket_bk_r [{shapeType=Capsule, length=15, radius=3, shapeName=dyn_jacket_bk_r_capsule, center=<15 0 0>, rotation=<rot 0 0 90>}]
 dyn_jacket_bk_r_base [{shapeType=Capsule, length=2, radius=1, shapeName=dyn_jacket_bk_r_base_capsule}]
 dyn_jacket_fr_l [
  {shapeType=Capsule, length=15, radius=3, shapeName=dyn_jacket_fr_l_capsule, center=<13 0 2.2>, rotation=<rot 0.0005 0.0003 90>}
  {
   shapeType=Capsule
   length=15
   radius=3
   shapeName=dyn_jacket_fr_l_capsule
   center=<13.1905 -11.2318 -3.1362>
   rotation=<rot 0.0007 0.0004 45>
  }
 ]
 dyn_jacket_fr_l_base [{shapeType=Capsule, length=2, radius=1, shapeName=dyn_jacket_fr_l_base_capsule}]
 dyn_jacket_fr_r [
  {shapeType=Capsule, length=14, radius=3, shapeName=dyn_jacket_fr_r_capsule, center=<13 0 2.2>, rotation=<rot 0.0003 -0.0002 85>}
  {
   shapeType=Capsule
   length=14
   radius=3
   shapeName=dyn_jacket_fr_r_capsule
   center=<12.4915 11.8523 -2.1472>
   rotation=<rot 0.0006 -0.0004 130>
  }
 ]
 dyn_jacket_fr_r_base [{shapeType=Capsule, length=2, radius=1, shapeName=dyn_jacket_fr_r_base_capsule}]
 dyn_jacket_side_l [{shapeType=Capsule, length=15, radius=3, shapeName=dyn_jacket_side_l_capsule, center=<13 0 0>, rotation=<rot 0 0 90>}]
 dyn_jacket_side_l_base [{shapeType=Capsule, length=2, radius=1, shapeName=dyn_jacket_side_l_base_capsule}]
 dyn_jacket_side_r [{shapeType=Capsule, length=16.5, radius=3, shapeName=dyn_jacket_side_r_capsule, center=<13 0 2.2>, rotation=<rot 0 0 90>}]
 dyn_jacket_side_r_base [{shapeType=Capsule, length=2, radius=1, shapeName=dyn_jacket_side_r_base_capsule}]
 extra_pocket_l [{
   shapeType=Capsule
   length=15
   radius=7
   shapeName=extra_pocket_l_capsule
   center=<13.8178 0.0001 -5.3041>
   rotation=<rot -0.0003 89.9991 -109.9993>
  }]
 extra_pocket_r [{
   shapeType=Capsule
   length=15
   radius=7
   shapeName=extra_pocket_r_capsule
   center=<13.8178 -0.0001 -5.3041>
   rotation=<rot -0.0003 89.9991 70.0007>
  }]
 pelvis [
  {shapeType=Capsule, length=15, radius=10, center=<-2 -1.2979 0>}
  {shapeType=Capsule, length=15, radius=9.5, shapeName=pelvis_capsule, center=<-7.9911 -1.2979 0>}
 ]
 spine_01 [{shapeType=Capsule, length=13.5, radius=10}]
 thigh_l [{shapeType=Capsule, length=18.5, radius=8.7, center=<-20 0.005 0>, rotation=<rot -90 -0.0173 0>}]
 thigh_r [{shapeType=Capsule, length=18.5, radius=8.7, center=<20 -0.005 0>, rotation=<rot -90 179.9827 0>}]
}
constraints [
 {bone1Name=spine_01, bone2Name=pelvis, swing1LimitDegrees=45, swing2LimitDegrees=45, twistLimitDegrees=45}
 {
  bone1Name=dyn_jacket_bk_r_base
  bone2Name=spine_01
  swing1LimitDegrees=45
  swing1Motion=Locked
  swing2LimitDegrees=45
  swing2Motion=Locked
  twistLimitDegrees=45
  twistMotion=Locked
 }
 {bone1Name=dyn_jacket_bk_r, bone2Name=dyn_jacket_bk_r_base, swing1LimitDegrees=20, swing2LimitDegrees=50, twistLimitDegrees=5}
 {
  bone1Name=dyn_jacket_fr_r_base
  bone2Name=spine_01
  swing1LimitDegrees=45
  swing1Motion=Locked
  swing2LimitDegrees=45
  swing2Motion=Locked
  twistLimitDegrees=45
  twistMotion=Locked
 }
 {bone1Name=dyn_jacket_fr_r, bone2Name=dyn_jacket_fr_r_base, swing1LimitDegrees=20, swing2LimitDegrees=30, twistLimitDegrees=5}
 {
  bone1Name=dyn_jacket_side_r_base
  bone2Name=spine_01
  swing1LimitDegrees=45
  swing1Motion=Locked
  swing2LimitDegrees=45
  swing2Motion=Locked
  twistLimitDegrees=45
  twistMotion=Locked
 }
 {bone1Name=dyn_jacket_side_r, bone2Name=dyn_jacket_side_r_base, swing1LimitDegrees=50, swing2LimitDegrees=30, twistLimitDegrees=5}
 {bone1Name=thigh_l, bone2Name=pelvis, swing1LimitDegrees=45, swing2LimitDegrees=45, twistLimitDegrees=45}
 {bone1Name=thigh_r, bone2Name=pelvis, swing1LimitDegrees=45, swing2LimitDegrees=45, twistLimitDegrees=45}
 {bone1Name=dyn_jacket_fr_l, bone2Name=dyn_jacket_fr_l_base, swing1LimitDegrees=20, swing2LimitDegrees=30, twistLimitDegrees=5}
 {
  bone1Name=dyn_jacket_fr_l_base
  bone2Name=spine_01
  swing1LimitDegrees=45
  swing1Motion=Locked
  swing2LimitDegrees=45
  swing2Motion=Locked
  twistLimitDegrees=45
  twistMotion=Locked
 }
 {bone1Name=dyn_jacket_side_l, bone2Name=dyn_jacket_side_l_base, swing1LimitDegrees=40, swing2LimitDegrees=30, twistLimitDegrees=5}
 {
  bone1Name=dyn_jacket_side_l_base
  bone2Name=spine_01
  swing1LimitDegrees=45
  swing1Motion=Locked
  swing2LimitDegrees=45
  swing2Motion=Locked
  twistLimitDegrees=45
  twistMotion=Locked
 }
 {bone1Name=dyn_jacket_bk_l, bone2Name=dyn_jacket_bk_l_base, swing1LimitDegrees=20, swing2LimitDegrees=50, twistLimitDegrees=5}
 {
  bone1Name=dyn_jacket_bk_l_base
  bone2Name=spine_01
  swing1LimitDegrees=45
  swing1Motion=Locked
  swing2LimitDegrees=45
  swing2Motion=Locked
  twistLimitDegrees=45
  twistMotion=Locked
 }
 {
  bone1Name=extra_pocket_r
  bone2Name=pelvis
  swing1LimitDegrees=45
  swing1Motion=Free
  swing2LimitDegrees=45
  swing2Motion=Free
  twistLimitDegrees=45
  twistMotion=Free
 }
 {
  bone1Name=extra_pocket_l
  bone2Name=pelvis
  swing1LimitDegrees=45
  swing1Motion=Free
  swing2LimitDegrees=45
  swing2Motion=Free
  twistLimitDegrees=45
  twistMotion=Free
 }
]
