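# Animation Blueprint for the CreamSkull DoveBee male body cosmetic, adding procedural and physics motion to a pose copied from the parent mesh.
# Weighted multi-bone constraints drive both pocket bones, then a Control Rig fed crouch, targeting, front-end and emote flags caches the pose.
# Two spine-relative RigidBody simulations with standard and crouch physics assets blend by bool into a post-dynamics Control Rig.
# The EventGraph updates those flags and computes a gravity override using a deprecated helper.
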AnimBlueprint /BRCosmetics/Characters/Player/Male/Medium/Bodies/M_MED_CreamSkull_DoveBee/Meshes/M_MED_CreamSkull_DoveBee_AnimBP
target_skeleton=&/M_MED_CreamSkull_DoveBee_Skeleton
graphs [
 AnimGraph {nodes [
   n1 = RigidBody!(
    ComponentPose=Local To Component(LocalPose=Use cached pose 'PreDyn'())
    bAlphaBoolEnabled:bool
    OverrideWorldGravity:Vector="0,0,0"
    ComponentLinearAccScale:Vector="0,0,0"
    ComponentLinearVelScale:Vector="0,0,0"
    bTransferBoneVelocities:bool=False
   ) -> (Pose) {
    alphaInputType=Bool
    bClampLinearTranslationLimitToRefPose=true
    bFreezeIncomingPoseOnStart=true
    bOverrideWorldGravity=true
    bTransferBoneVelocities=true
    overridePhysicsAsset=&/M_MED_CreamSkull_DoveBee_Physics
    simulationSpace=BaseBoneSpace
    alphaBoolBlend {blendOption=Linear}
    baseBoneRef {boneName=spine_02}
    componentAppliedLinearAccClamp=<100 100 100>
    componentLinearAccScale=<0.6 0.2 0.2>
    componentLinearVelScale=<0.6 0.2 0.2>
    simSpaceSettings {
     dampingAlpha=0
     maxAngularAcceleration=50
     maxAngularVelocity=8
     maxLinearAcceleration=2000
     maxLinearVelocity=500
     worldAlpha=1
     externalLinearDragV=<1 1 1>
    }
   }
   n2 = RigidBody(
    ComponentPose=Local To Component(LocalPose=Use cached pose 'PreDyn'())
    OverrideWorldGravity="0,0,0"
    ComponentLinearAccScale="0,0,0"
    ComponentLinearVelScale="0,0,0"
    bTransferBoneVelocities=False
   ) {
    alphaInputType=Bool
    bClampLinearTranslationLimitToRefPose=true
    bFreezeIncomingPoseOnStart=true
    bOverrideWorldGravity=true
    bTransferBoneVelocities=true
    overridePhysicsAsset=&/M_MED_CreamSkull_DoveBee_Physics_crouch
    simulationSpace=BaseBoneSpace
    alphaBoolBlend {blendOption=Linear}
    baseBoneRef {boneName=spine_02}
    componentAppliedLinearAccClamp=<100 100 100>
    componentLinearAccScale=<0.6 0.2 0.2>
    componentLinearVelScale=<0.6 0.2 0.2>
    simSpaceSettings {
     dampingAlpha=0
     maxAngularAcceleration=50
     maxAngularVelocity=8
     maxLinearAcceleration=2000
     maxLinearVelocity=500
     worldAlpha=1
     externalLinearDragV=<1 1 1>
    }
   }
   n3 = Control Rig(Source=Blend Poses by bool(BlendPose_0=Component To Local(ComponentPose=@n2), BlendPose_1=Component To Local(ComponentPose=@n1))) {controlRigAssetReference {blueprintRigClass=&/M_MED_CreamSkull_DoveBee_CtrlRig_PostDyn.M_MED_CreamSkull_DoveBee_CtrlRig_PostDyn_C}}
   Output Pose(Result=@n3) {layerGroup=None}
   n4 = "Constraint - extra_pocket_r (PPPPPPPP)"!(
    ComponentPose=Local To Component(LocalPose=Copy Pose From Mesh() {bUseAttachedParent=true})
    Alpha=1
    ConstraintWeights_0=0.1163
    ConstraintWeights_1:float=0
    ConstraintWeights_2:float=0
    ConstraintWeights_3:float=0.0004
    ConstraintWeights_4:float=0.0471
    ConstraintWeights_5:float=0.6862
    ConstraintWeights_6:float=0.0014
    ConstraintWeights_7:float=0.1485
   ) -> (Pose) {
    boneToModify {boneName=extra_pocket_r}
    constraintSetup [
     {targetBone {boneName=pelvis}}
     {targetBone {boneName=deform_pec_r}}
     {targetBone {boneName=deform_pec_l}}
     {targetBone {boneName=thigh_twist_01_l}}
     {targetBone {boneName=thigh_r}}
     {targetBone {boneName=thigh_twist_01_r}}
     {targetBone {boneName=deform_glute_r}}
     {targetBone {boneName=deform_groin_r}}
    ]
    constraintWeights [1,1,1,1,1,1,1,1]
   }
   n5 = "Constraint - extra_pocket_l (PPPPPPPP)"!(
    ComponentPose=@n4
    Alpha=1
    ConstraintWeights_0=0.1163
    ConstraintWeights_1:float=0
    ConstraintWeights_2:float=0
    ConstraintWeights_3:float=0.0471
    ConstraintWeights_4:float=0.6862
    ConstraintWeights_5:float=0.0014
    ConstraintWeights_6:float=0.1485
    ConstraintWeights_7:float=0.0004
   ) -> (Pose) {
    boneToModify {boneName=extra_pocket_l}
    constraintSetup [
     {targetBone {boneName=pelvis}}
     {targetBone {boneName=deform_pec_r}}
     {targetBone {boneName=deform_pec_l}}
     {targetBone {boneName=thigh_l}}
     {targetBone {boneName=thigh_twist_01_l}}
     {targetBone {boneName=deform_glute_l}}
     {targetBone {boneName=deform_groin_l}}
     {targetBone {boneName=thigh_twist_01_r}}
    ]
    constraintWeights [1,1,1,1,1,1,1,1]
   }
   n6 = Control Rig!(
    Source=Component To Local(ComponentPose=@n5)
    Alpha
    crouch:bool=Get Crouching() -> (Crouching:bool)
    target:bool=Get Targeting() -> (Targeting:bool)
    FE:bool=Get FE() -> (FE:bool)
    emote:bool=Get emote() -> (emote:bool)
   ) -> (Pose) {controlRigAssetReference {blueprintRigClass=&/M_MED_CreamSkull_DoveBee_CtrlRig.M_MED_CreamSkull_DoveBee_CtrlRig_C}}
   Save cached pose 'PreDyn'(Pose=@n6) {cacheName=PreDyn}
  ]}
 EventGraph {nodes [
   Event BlueprintUpdateAnimation()
   > n1 = Sequence!(execute) -> (then_0, then_1, then_2:Exec, then_3:Exec, then_4:Exec)
   Set Targeting(execute=@n1.then_1, Targeting:bool=Get bIsTargeting()) -> (Output_Get:bool)
   n2 = Get bIsCrouching()
   Set Crouching(execute=@n1.then_2, Crouching:bool=@n2) -> (Output_Get:bool)
   > Set Dyn(Dyn:bool=NOT Boolean(A=@n2)) -> (Output_Get:bool)
   n3 = Gravity Override State Function DEPRECATED(
    execute=@n1.then_0
    CharacterPartAnimInstance=Self-Reference()
    GameTimeInSeconds=GetGameTimeInSeconds()
    GravityOverrideParams=Get Gravity Override Params()
    DebugJointName=spine_05
   )
   > Set GravityOverride(GravityOverride=@n3)
   Set FE(execute=@n1.then_3, FE:bool=Get bIsInFrontEnd()) -> (Output_Get:bool)
   Set emote(execute=@n1.then_4, emote:bool=Get bIsPlayingEmote()) -> (Output_Get:bool)
  ]}
]
variables [
 Crouching {type=boolean}
 Dyn {type=boolean}
 emote {type=boolean}
 FE {type=boolean}
 Gravity Override Params {type=GravityOverrideParamsStruct}
 GravityOverride {type=Vector}
 Targeting {type=boolean}
]
