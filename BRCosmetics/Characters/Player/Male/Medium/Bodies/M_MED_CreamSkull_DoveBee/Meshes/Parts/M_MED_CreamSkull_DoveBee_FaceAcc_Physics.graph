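# Physics asset for the face accessory part of the male medium CreamSkull DoveBee player cosmetic, defining capsule bodies for the head and three hair chains.
# Left, middle and right base hair bodies are rotation-locked to the head and act as anchors.
# Each dynamic hair body hangs from its base with limited swing and tight twist, with the middle strand allowed less swing than the sides.
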PhysicsAsset /BRCosmetics/Characters/Player/Male/Medium/Bodies/M_MED_CreamSkull_DoveBee/Meshes/Parts/M_MED_CreamSkull_DoveBee_FaceAcc_Physics
bodies {
 base_hair_l [{shapeType=Capsule, length=2, radius=1, shapeName=base_hair_l_capsule}]
 base_hair_m [{shapeType=Capsule, length=2, radius=1, shapeName=base_hair_m_capsule}]
 base_hair_r [{shapeType=Capsule, length=2, radius=1, shapeName=base_hair_r_capsule}]
 dyn_hair_l [{shapeType=Capsule, length=5, radius=2, shapeName=dyn_hair_l_capsule, center=<5 0 0>, rotation=<rot 0 90 0>}]
 dyn_hair_m [{shapeType=Capsule, length=5, radius=2, shapeName=dyn_hair_m_capsule, center=<5 0 0>, rotation=<rot 0 90 0>}]
 dyn_hair_r [{shapeType=Capsule, length=5, radius=2, shapeName=dyn_hair_r_capsule, center=<5 0 0>, rotation=<rot 0 90 0>}]
 head [{shapeType=Capsule, length=2, radius=1, shapeName=head_capsule}]
}
constraints [
 {bone1Name=dyn_hair_r, bone2Name=base_hair_r, swing1LimitDegrees=60, swing2LimitDegrees=30, twistLimitDegrees=5}
 {
  bone1Name=base_hair_l
  bone2Name=head
  swing1LimitDegrees=45
  swing1Motion=Locked
  swing2LimitDegrees=45
  swing2Motion=Locked
  twistLimitDegrees=45
  twistMotion=Locked
 }
 {bone1Name=dyn_hair_l, bone2Name=base_hair_l, swing1LimitDegrees=60, swing2LimitDegrees=30, twistLimitDegrees=5}
 {
  bone1Name=base_hair_m
  bone2Name=head
  swing1LimitDegrees=45
  swing1Motion=Locked
  swing2LimitDegrees=45
  swing2Motion=Locked
  twistLimitDegrees=45
  twistMotion=Locked
 }
 {bone1Name=dyn_hair_m, bone2Name=base_hair_m, swing1LimitDegrees=30, swing2LimitDegrees=30, twistLimitDegrees=5}
 {
  bone1Name=base_hair_r
  bone2Name=head
  swing1LimitDegrees=45
  swing1Motion=Locked
  swing2LimitDegrees=45
  swing2Motion=Locked
  twistLimitDegrees=45
  twistMotion=Locked
 }
]
